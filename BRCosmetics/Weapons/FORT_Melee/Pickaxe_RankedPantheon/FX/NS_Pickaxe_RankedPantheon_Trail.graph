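# Niagara swing trail effect for the RankedPantheon pickaxe, built from three ribbon emitters with two renderers each.
# Each emitter uses AnimTrail modules to spawn and update short-lived particles between its own pair of skeletal mesh bones or sockets.
# Curves fade color and alpha and taper ribbon width over particle life.
# The green-tinted ColorTrail brightens at night via time-of-day selection, and a gradient texture is exposed as a user parameter.
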NiagaraSystem /BRCosmetics/Weapons/FORT_Melee/Pickaxe_RankedPantheon/FX/NS_Pickaxe_RankedPantheon_Trail
dependencies {
 usedDataInterfaces [&NiagaraDataInterfaceCurve, &NiagaraDataInterfaceSkeletalMesh]
 usedDynamicInputs []
 usedModules [
  &SystemState
  &self:SystemSpawnScript.SystemScriptSource.SystemScriptGraph.NiagaraNodeAssignment_0.SetVariables_#1
 ]
}
emitters {
 ColorTrail {
  emitterUpdateScript [
   &EmitterState(Life Cycle Mode=::NewEnumerator1, Loop Behavior=::NewEnumerator1, Loop Duration=0.3, UseLoopDelay=false)
   &FN_SpawnAnimTrail()
  ]
  particleSpawnScript [
   &FN_InitializeAnimTrailBoneIndecies(First Bone or Socket Index=1, Second Bone or Socket Index=0)
   &FN_InitializeAnimTrail(Color=<rgba 0.0152 0.4678 0.1683 1>, Lifetime=0.1, SkeletalMesh=$System.Skeletal Mesh)
   &AddVelocity(Velocity=<0 -100 0>) {enabled=false}
  ]
  particleUpdateScript [
   &FN_UpdateAnimTrail(SkeletalMesh=$System.Skeletal Mesh) {module=FN_UpdateAnimTrail001}
   &ScaleColor(
    Scale Alpha=&FloatFromCurve(
     FloatCurve=DataInterface() {propertyValues {Curve {keys [
         {arriveTangent=-0.5448, interpMode=RCIM_Cubic, leaveTangent=-0.5448, tangentMode=RCTM_User, time=0, value=1}
         {arriveTangent=-0.643, interpMode=RCIM_Cubic, leaveTangent=-0.643, tangentMode=RCTM_User, time=1, value=0}
        ]}}}
     Scale Curve=0.25
    )
    Scale RGB=&Multiply_VectorByFloat(
     Vector=<1 1 1>
     Float=&FloatFromCurve(
      FloatCurve=DataInterface() {propertyValues {Curve {keys [{tangentMode=RCTM_User, time=0.5, value=1}, {tangentMode=RCTM_User, time=1, value=0.5}]}}}
      Scale Curve=&FN_Select_Float_from_Time_Of_Day(Morning=1, Day=1, Evening=1, Night=2.5)
     )
    )
    ScaleA=true
    ScaleRGB=true
   )
   /SetVariables(Particles.RibbonV0RangeOverride:Vector2f=<1 0>) {isSetParameters=true, module=SetVariables_#2}
   &ScaleRibbonWidth(Ribbon Width Scale=&FloatFromCurve(FloatCurve=DataInterface() {propertyValues {Curve {keys [<at 0.75 1>,<at 1 0>]}}}))
  ]
  rendererClasses [&NiagaraRibbonRendererProperties]
  renderers [
   {rendererClass=&NiagaraRibbonRendererProperties}
   {rendererClass=&NiagaraRibbonRendererProperties, rendererIndex=1}
  ]
 }
 ColorTrail001 {
  emitterUpdateScript [
   &EmitterState(Life Cycle Mode=::NewEnumerator1, Loop Behavior=::NewEnumerator1, Loop Duration=0.3, UseLoopDelay=false)
   &FN_SpawnAnimTrail()
  ]
  particleSpawnScript [
   &FN_InitializeAnimTrailBoneIndecies(First Bone or Socket Index=2, Second Bone or Socket Index=3)
   &FN_InitializeAnimTrail(Lifetime=0.1, SkeletalMesh=$System.Skeletal Mesh)
   &AddVelocity(Velocity=<0 -100 0>) {enabled=false}
  ]
  particleUpdateScript [
   &FN_UpdateAnimTrail(SkeletalMesh=$System.Skeletal Mesh) {module=FN_UpdateAnimTrail001}
   &ScaleColor(
    Scale Alpha=&FloatFromCurve(
     FloatCurve=DataInterface() {propertyValues {Curve {keys [<at 0.1 1>, {arriveTangent=-0.643, interpMode=RCIM_Cubic, leaveTangent=-0.643, tangentMode=RCTM_User, time=1, value=0}]}}}
     Scale Curve=0.05
    )
    Scale RGB=&Multiply_VectorByFloat(
     Vector=<1 1 1>
     Float=&FloatFromCurve(FloatCurve=DataInterface() {propertyValues {Curve {keys [{tangentMode=RCTM_User, time=0.5, value=1}, {tangentMode=RCTM_User, time=1, value=0.5}]}}})
    )
    ScaleA=true
    ScaleRGB=true
   )
   /SetVariables(Particles.RibbonV0RangeOverride:Vector2f=<1 0>) {isSetParameters=true, module=SetVariables_#2}
   &ScaleRibbonWidth(Ribbon Width Scale=&FloatFromCurve(FloatCurve=DataInterface() {propertyValues {Curve {keys [<at 0.75 1>,<at 1 0>]}}}))
  ]
  rendererClasses [&NiagaraRibbonRendererProperties]
  renderers [
   {rendererClass=&NiagaraRibbonRendererProperties}
   {rendererClass=&NiagaraRibbonRendererProperties, rendererIndex=1}
  ]
 }
 ColorTrail002 {
  emitterUpdateScript [
   &EmitterState(Life Cycle Mode=::NewEnumerator1, Loop Behavior=::NewEnumerator1, Loop Duration=0.3, UseLoopDelay=false)
   &FN_SpawnAnimTrail()
  ]
  particleSpawnScript [
   &FN_InitializeAnimTrailBoneIndecies(First Bone or Socket Index=4, Second Bone or Socket Index=5)
   &FN_InitializeAnimTrail(Lifetime=0.1, SkeletalMesh=$System.Skeletal Mesh)
   &AddVelocity(Velocity=<0 -100 0>) {enabled=false}
  ]
  particleUpdateScript [
   &FN_UpdateAnimTrail(SkeletalMesh=$System.Skeletal Mesh) {module=FN_UpdateAnimTrail001}
   &ScaleColor(
    Scale Alpha=&FloatFromCurve(
     FloatCurve=DataInterface() {propertyValues {Curve {keys [<at 0.1 1>, {arriveTangent=-0.643, interpMode=RCIM_Cubic, leaveTangent=-0.643, tangentMode=RCTM_User, time=1, value=0}]}}}
     Scale Curve=0.05
    )
    Scale RGB=&Multiply_VectorByFloat(
     Vector=<1 1 1>
     Float=&FloatFromCurve(
      FloatCurve=DataInterface() {propertyValues {Curve {keys [{tangentMode=RCTM_User, time=0.5, value=1}, {tangentMode=RCTM_User, time=1, value=0.5}]}}}
      Scale Curve=&FN_Select_Float_from_Time_Of_Day(Morning=1, Day=1, Evening=1, Night=1)
     )
    )
    ScaleA=true
    ScaleRGB=true
   )
   /SetVariables(Particles.RibbonV0RangeOverride:Vector2f=<1 0>) {isSetParameters=true, module=SetVariables_#2}
   &ScaleRibbonWidth(Ribbon Width Scale=&FloatFromCurve(FloatCurve=DataInterface() {propertyValues {Curve {keys [<at 0.75 1>,<at 1 0>]}}}))
  ]
  rendererClasses [&NiagaraRibbonRendererProperties]
  renderers [
   {rendererClass=&NiagaraRibbonRendererProperties}
   {rendererClass=&NiagaraRibbonRendererProperties, rendererIndex=1}
  ]
 }
}
summary {
 systemName=NS_Pickaxe_RankedPantheon_Trail
 emitters [
  {emitterName=ColorTrail, rendererClasses [&NiagaraRibbonRendererProperties]}
  {emitterName=ColorTrail001, rendererClasses [&NiagaraRibbonRendererProperties]}
  {emitterName=ColorTrail002, rendererClasses [&NiagaraRibbonRendererProperties]}
 ]
 userVariables [
  User.GradientTexture {
   type=Texture
   description=""
   defaultValue {
    struct=NiagaraExt_VariableValue_Object
    value {object=&/T_FX_Pickaxe_RankedPantheon_Trail, objectClass=/Script/Engine.Texture}
   }
  }
  User.RibbonIDOffset {type=NiagaraInt32}
 ]
}
