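# Animation Blueprint for the RankedPantheon pickaxe that animates the floating orb on the axe.
# The update graph accumulates a wrapping yaw spin, sped up in skydive dive mode, and computes a time-driven random drift vector.
# The AnimGraph applies spin to dyn_orb_02 and drift to dyn_orb_01, then adds spring-based Anim Dynamics and a fixed offset on dyn_orb_mstr.
# A second Anim Dynamics node is unconnected, and the interpolated Speed Alpha feeds nothing downstream.
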AnimBlueprint /BRCosmetics/Weapons/FORT_Melee/Pickaxe_RankedPantheon/Meshes/RankedPantheon_Axe_AnimBP
target_skeleton=&/RankedPantheon_Axe_Skeleton
graphs [
 AnimGraph {nodes [
   Anim Dynamics - Bone: dyn_orb_mstr(Alpha=0.25, ExternalForce="0,0,0") {
    angularDampingOverride=1
    angularSpringConstant=20000
    bAngularSpring=true
    bGravityOverrideInSimSpace=true
    bLinearSpring=true
    bOverrideAngularDamping=true
    bOverrideLinearDamping=true
    bUseGravityOverride=true
    bUsePlanarLimit=false
    gravityScale=0.35
    linearDampingOverride=0.9999
    linearSpringConstant=50
    simulationSpace=BoneRelative
    alphaBoolBlend {blendOption=Linear}
    boundBone {boneName=dyn_orb_mstr}
    componentAppliedLinearAccClamp=<30000 30000 30000>
    componentLinearAccScale=<0.3 0.3 0.4>
    componentLinearVelScale=<0.3 0.3 0.4>
    gravityOverride {z=-980}
    physicsBodyDefinitions [{
      sphereCollisionRadius=0
      boundBone {boneName=dyn_orb_mstr}
      boxExtents=<5 10 10>
      constraintSetup {angularLimitsMax=<0 0 0>, angularLimitsMin=<0 0 0>, angularTarget=<1 0 0>, linearAxesMax=<15 10 10>, linearAxesMin=<-15 -10 -5>}
      localJointOffset=<0 0 1>
     }]
    relativeSpaceBone {boneName=b_AR_Root}
   }
   AnimGraphNode_SubInput_0 = Input Pose()
   n1 = "Transform (Modify) Bone - Bone: dyn_orb_02"(
    ComponentPose=Local To Component(LocalPose=@AnimGraphNode_SubInput_0)
    Alpha=1
    Translation="0,0,0"
    Rotation=Get Spin Result A() -> (Spin Result A:Rotator)
    Scale="1,1,1"
   ) {rotationMode=BMM_Additive, rotationSpace=BCS_BoneSpace, alphaBoolBlend {blendOption=Linear}, boneToModify {boneName=dyn_orb_02}}
   n2 = "Transform (Modify) Bone - Bone: dyn_orb_01"(ComponentPose=@n1, Alpha=1, Translation=Get Globe movement() -> (Globe movement:Vector), Rotation="0,0,0", Scale="1,1,1") {
    rotationSpace=BCS_WorldSpace
    translationMode=BMM_Additive
    translationSpace=BCS_BoneSpace
    alphaBoolBlend {blendOption=Linear}
    alphaScaleBiasClamp {interpSpeedDecreasing=2, interpSpeedIncreasing=2}
    boneToModify {boneName=dyn_orb_01}
   }
   n3 = Anim Dynamics - Bone: dyn_orb_mstr(ComponentPose=@n2, Alpha=0.2, ExternalForce="0,0,0") {
    angularDampingOverride=1
    angularSpringConstant=2000
    bGravityOverrideInSimSpace=true
    bLinearSpring=true
    bOverrideAngularDamping=true
    bOverrideLinearDamping=true
    bUseGravityOverride=true
    bUsePlanarLimit=false
    gravityScale=0.35
    linearDampingOverride=1
    linearSpringConstant=100
    alphaBoolBlend {blendOption=Linear}
    boundBone {boneName=dyn_orb_mstr}
    componentAppliedLinearAccClamp=<6000 6000 6000>
    componentLinearAccScale=<0.4 0.6 0.4>
    componentLinearVelScale=<0.5 0.5 0.7>
    gravityOverride {z=-980}
    physicsBodyDefinitions [{
      sphereCollisionRadius=0
      boundBone {boneName=dyn_orb_mstr}
      boxExtents=<100 100 100>
      constraintSetup {angularLimitsMax=<0 0 0>, angularLimitsMin=<0 0 0>, angularTarget=<1 0 0>, linearAxesMax=<30 30 30>, linearAxesMin=<-30 -30 -30>}
      localJointOffset=<0 0 -3>
     }]
    relativeSpaceBone {boneName=b_AR_Root}
   }
   n4 = "Transform (Modify) Bone - Bone: dyn_orb_mstr"(ComponentPose=@n3, Alpha=1, Translation="0,0,5", Rotation="0,0,0", Scale="1,1,1") {
    rotationSpace=BCS_WorldSpace
    translationMode=BMM_Additive
    translationSpace=BCS_BoneSpace
    alphaBoolBlend {blendOption=Linear}
    alphaScaleBiasClamp {interpSpeedDecreasing=2, interpSpeedIncreasing=2}
    boneToModify {boneName=dyn_orb_mstr}
    translation {z=5}
   }
   Output Pose(Result=Component To Local(ComponentPose=@n4)) {layerGroup=None}
   Get GlobeLocal() -> (GlobeLocal:Rotator)
  ]}
 EventGraph {nodes [
   n1 = Event BlueprintUpdateAnimation()
   > n2 = Sequence!(execute) -> (then_0, then_1, then_2:Exec, then_3:Exec)
   > Set Delta Time Remap(Delta Time Remap:double=FInterpTo(Current=Get Delta Time Remap() -> (Delta Time Remap:double), Target=1, DeltaTime=0, InterpSpeed=1)) -> (Output_Get:double)
   n3 = Spin Function!(
    execute=@n2.then_1
    self
    Spin:double=Get Spin A() -> (Spin A:double)
    Speed:double=Get Speed A() -> (Speed A:double)
    Velocity:double=Get Velocity A() -> (Velocity A:double)
   ) -> (then, Output:double)
   > n4 = Set Spin A(Spin A:double=@n3) -> (Output_Get:double)
   > Set Spin Result A(Spin Result A:Rotator=MakeRotator(Yaw=@n4)) -> (Output_Get:Rotator)
   n5 = Random Vector Function!(
    execute=@n2.then_2
    CharacterPartAnimInstance:Custom Character Part Anim Instance Object Reference
    SkelMeshComponent:Mesh Component Object Reference=GetOwnerSkeletalMeshComponent()
    JointName:Name=None
    GameTimeInSeconds:double=GetGameTimeInSeconds()
    TimeOffset:double=0
    PerlinScale:double=0
    Frequency:Vector="1.3,2,1"
    Amplitude:Vector="1.2,1.2,2.2"
    Offset:Vector="0,0,0"
   ) -> (then, VectorOut:Vector)
   > Set Globe movement(Globe movement:Vector=@n5) -> (Output_Get:Vector)
   n6 = OR Boolean!(A=Get bIsFalling(), B=Get bIsSkyDiving(), C:bool=Get TacSprint() -> (TacSprint:bool)) -> (ReturnValue)
   Set Speed Alpha(
    execute=@n2.then_3
    Speed Alpha:double=FInterpTo(Current=Get Speed Alpha() -> (Speed Alpha:double), Target=SelectFloat(A=1, B=0, bPickA=@n6), DeltaTime=@n1.DeltaTimeX, InterpSpeed=4)
   ) -> (Output_Get:double)
   n7 = Get bIsSkydiveDiveMode()
   > Set Velocity A(Velocity A:double=float * float(
     A=FInterpEaseInOut!(A:double=1.15, B:double=6.4, Alpha:double=0, Exponent:double=3) -> (ReturnValue:double)
     B=SelectFloat(A=1.75, B=1, bPickA=@n7)
    ) {is=title K2Node_PromotableOperator}) -> (Output_Get:double)
   GetRealTimeSeconds()
   TryGetPawnOwner()
   Reroute Node(InputPin:Wildcard) -> (OutputPin:Wildcard)
   Get bIsSprinting()
  ]}
 Spin Function {nodes [
   n1 = Spin Function!() -> (then, Spin:double=~, Speed:double=~, Velocity:double=~) {is=title K2Node_FunctionEntry}
   > Set Velocity Mult(Velocity Mult:double=@n1.Velocity) -> (Output_Get:double)
   > n2 = Set Spin Var(Spin Var:double=@n1) -> (Output_Get:double)
   > n3 = Branch(Condition="float >= float"(A=@n2, B=720) {is=title K2Node_CallFunction})
   > n4 = Set Spin Var(Spin Var=float - float(A=Get Spin Var() -> (Spin Var:double), B=720) {is=title K2Node_CallFunction})
   n5 = Get Velocity Mult() -> (Velocity Mult:double)
   n6 = Get Delta Time Remap()
   n7 = Set Spin Var(
    execute=@n3.else
    Spin Var=float + float!(
     A=Get Spin Var()
     B=float * float(A=@n1.Speed, B=@n6) {is=title K2Node_CommutativeAssociativeBinaryOperator}
     C:double=float * float(A=1, B=@n5)
    ) -> (ReturnValue) {is=title K2Node_CommutativeAssociativeBinaryOperator}
   )
   Return Node!(execute=[@n7.then, @n4.then], Output:double=@n7)
  ]}
]
variables [
 Delta Time Remap {type=number}
 Globe movement {type=Vector}
 Speed A {type=number}
 Speed Alpha {type=number}
 Spin A {type=number}
 Spin Result A {type=Rotator}
 Spin Var {type=number}
 TacSprint {type=boolean}
 Velocity A {type=number}
 Velocity Mult {type=number}
]
